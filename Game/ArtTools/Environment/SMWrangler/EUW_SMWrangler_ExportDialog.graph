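# Editor Utility dialog for SMWrangler that exports a list of generated LODs to the clipboard or a CSV file.
# Construct labels the count from a temp file via Python, while button enabling uses the separate LODCount variable, never set here.
# Buttons run inline Python that reads the temp-file list and shells out to PowerShell for Set-Clipboard or a Windows save dialog.
# Both buttons and cancellation delete a temp lock file.
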EditorUtilityWidgetBlueprint /Game/ArtTools/Environment/SMWrangler/EUW_SMWrangler_ExportDialog
graphs [EventGraph {nodes [
   "On Clicked (BTN_ExportToCSV)"() {is=title K2Node_ComponentBoundEvent}
   > Execute Python Script(PythonScript="import unreal, subprocess, pathlib, os, tempfile; tmp = os.path.join(tempfile.gettempdir(), 'smwrangler_meshnames.txt'); content = pathlib.Path(tmp).read_text(encoding='utf-8') if os.path.exists(tmp) else ''; unreal.log('SMWQ csv content_len=' + str(len(content))); ps = \"Add-Type -AssemblyName System.Windows.Forms; $d=New-Object System.Windows.Forms.SaveFileDialog; $d.Filter='CSV Files|*.csv|All Files|*.*'; $d.DefaultExt='csv'; $d.InitialDirectory=[Environment]::GetFolderPath('Desktop'); $d.Title='Choose CSV Export Location'; if($d.ShowDialog() -eq 'OK'){Write-Host $d.FileName}else{Write-Host 'CANCELLED'}\"; result = subprocess.run(['powershell','-NoProfile','-Command', ps], capture_output=True, text=True, timeout=120); p = result.stdout.strip(); p = '' if p == 'CANCELLED' else p; unreal.log('SMWQ csv path=' + p); pathlib.Path(p).write_text(content, encoding='utf-8') if (p and content) else unreal.log('SMWQ csv skipped p=' + str(bool(p)) + ' c=' + str(bool(content))); lock = os.path.join(tempfile.gettempdir(), 'smwrangler_dialog.lock'); os.remove(lock) if os.path.exists(lock) else None; unreal.log('SMWQ lock cleared')")
   "On Clicked (BTN_CopyToClipboard)"() {is=title K2Node_ComponentBoundEvent}
   > Execute Python Script(PythonScript="import unreal, os, pathlib, tempfile, subprocess; tmp = os.path.join(tempfile.gettempdir(), 'smwrangler_meshnames.txt'); content = pathlib.Path(tmp).read_text(encoding='utf-8') if os.path.exists(tmp) else ''; unreal.log('SMWQ clipboard content_len=' + str(len(content))); subprocess.run(['powershell','-NoProfile','-Command', 'Set-Clipboard -Value @\\'\\n' + content + '\\n\\'@'],     capture_output=True, text=True) if content else None; unreal.log('SMWQ clipboard done'); lock = os.path.join(tempfile.gettempdir(), 'smwrangler_dialog.lock'); os.remove(lock) if os.path.exists(lock) else None; unreal.log('SMWQ lock cleared')")
   Event Construct()
   > Execute Python Script(PythonScript="import unreal, os, tempfile, pathlib; cnt_file = os.path.join(tempfile.gettempdir(), 'smwrangler_count.txt'); count = int(pathlib.Path(cnt_file).read_text(encoding='utf-8').strip()) if os.path.exists(cnt_file) else 0; label = str(count) + (' Generated LOD' if count == 1 else ' Generated LODs'); unreal.log('SMWQ label=' + label); [w.set_text(unreal.Text(label)) for w in unreal.ObjectIterator(unreal.EditableText) if w.get_name() == 'TXT_LODCount']")
   n1 = "integer > integer"(A=Get LODCount() -> (LODCount:int=0))
   > SetIsEnabled(self=Get BTN_CopyToClipboard() -> (BTN_CopyToClipboard:Button Object Reference), bInIsEnabled=@n1)
   > SetIsEnabled(self=Get BTN_ExportToCSV() -> (BTN_ExportToCSV:Button Object Reference), bInIsEnabled=@n1)
   Event OnDialogCancelled() {is=title K2Node_Event}
   > Execute Python Script(PythonScript="import unreal, os, tempfile; lock = os.path.join(tempfile.gettempdir(), 'smwrangler_dialog.lock'); os.remove(lock) if os.path.exists(lock) else None; unreal.log('SMWQ lock cleared via X button')")
   Event PreConstruct()
   Event Tick()
  ]}]
variables [
 LODCount {type=integer}
 MeshNames {type=string, default=""}
 TXT_LODCount_0 {type=EditableText, default=null}
]
widgets {
 info {parentClass=/Script/Blutility.EditorUtilityDialogWidget, rootWidgetClass=&Border, widgetCount=8}
 widgets [
  RootBorder {
   widgetClassPath=&Border
   widget=RootBorder
   properties {
    background {
     bIsDynamicallyLoaded=false
     drawAs=Image
     imageType=NoImage
     mirroring=NoMirror
     resourceName=None
     resourceObject=null
     tiling=NoTile
     imageSize=<32 32>
     margin=<ltrb 0 0 0 0>
     outlineSettings {
      bUseBrushTransparency=false
      roundingType=HalfHeightRadius
      width=0
      color {colorUseRule=UseColor_Specified, specifiedColor=<rgba 0 0 0 0>}
      cornerRadii=<0 0 0 0>
     }
     tintColor {colorUseRule=UseColor_Specified, specifiedColor=<rgba 0.025 0.025 0.025 1>}
     uVRegion {bIsValid=false, max=<0 0>, min=<0 0>}
    }
    padding=<ltrb 16 16 16 16>
   }
  }
  OuterVBox {
   widgetClassPath=&VerticalBox
   parent=RootBorder
   widget=OuterVBox
   slot {class=BorderSlot, properties {padding=<ltrb 16 16 16 16>}}
  }
  TXT_DialogTitle {
   widgetClassPath=&TextBlock
   parent=OuterVBox
   widget=TXT_DialogTitle
   properties {
    text=Export Generated LODs List
    colorAndOpacity {colorUseRule=UseColor_Specified, specifiedColor=<rgba 0.9 0.9 0.9 1>}
    font {
     bForceMonospaced=false
     bMaterialIsStencil=false
     fontMaterial=null
     fontObject=&Roboto
     letterSpacing=0
     monospacedWidth=1
     size=16
     skewAmount=0
     typefaceFontName=Bold
     outlineSettings {
      bApplyOutlineToDropShadows=false
      bMiteredCorners=false
      bSeparateFillAlpha=false
      outlineMaterial=null
      outlineSize=0
      outlineColor=<rgba 0 0 0 1>
     }
    }
   }
   slot {class=VerticalBoxSlot, properties {horizontalAlignment=HAlign_Center, padding=<ltrb 0 0 0 12>}}
  }
  TXT_LODCount {
   widgetClassPath=/Script/UMG.EditableText
   bIsVariable=true
   parent=OuterVBox
   widget=TXT_LODCount
   properties {
    allowContextMenu=false
    isReadOnly=true
    justification=Center
    text=3 Generated LODs
    widgetStyle {
     backgroundImageComposing {
      bIsDynamicallyLoaded=false
      drawAs=Border
      imageType=FullColor
      mirroring=NoMirror
      resourceName=../../../Engine/Content/Slate/Old/HyperlinkDotted.png
      resourceObject=null
      tiling=Both
      imageSize=<0 0>
      margin=<ltrb 0 0 0 0.1875>
      outlineSettings {
       bUseBrushTransparency=false
       roundingType=HalfHeightRadius
       width=0
       color {colorUseRule=UseColor_Specified, specifiedColor=<rgba 0 0 0 0>}
       cornerRadii=<0 0 0 0>
      }
      tintColor {colorUseRule=UseColor_Specified, specifiedColor=<rgba 1 1 1 1>}
      uVRegion {bIsValid=false, max=<0 0>, min=<0 0>}
     }
     backgroundImageSelected {
      bIsDynamicallyLoaded=false
      drawAs=Image
      imageType=NoImage
      mirroring=NoMirror
      resourceName=None
      resourceObject=null
      tiling=NoTile
      imageSize=<0 0>
      margin=<ltrb 0 0 0 0>
      outlineSettings {
       bUseBrushTransparency=false
       roundingType=HalfHeightRadius
       width=0
       color {colorUseRule=UseColor_Specified, specifiedColor=<rgba 0 0 0 0>}
       cornerRadii=<0 0 0 0>
      }
      tintColor {colorUseRule=UseColor_Specified, specifiedColor=<rgba 0 0.162 0.7454 1>}
      uVRegion {bIsValid=false, max=<0 0>, min=<0 0>}
     }
     caretImage {
      bIsDynamicallyLoaded=false
      drawAs=Image
      imageType=NoImage
      mirroring=NoMirror
      resourceName=None
      resourceObject=null
      tiling=NoTile
      imageSize=<0 0>
      margin=<ltrb 0 0 0 0>
      outlineSettings {
       bUseBrushTransparency=false
       roundingType=HalfHeightRadius
       width=0
       color {colorUseRule=UseColor_Specified, specifiedColor=<rgba 0 0 0 0>}
       cornerRadii=<0 0 0 0>
      }
      tintColor {colorUseRule=UseColor_Specified, specifiedColor=<rgba 1 1 1 1>}
      uVRegion {bIsValid=false, max=<0 0>, min=<0 0>}
     }
     colorAndOpacity {colorUseRule=UseColor_Foreground, specifiedColor=<rgba 1 0 1 1>}
     font {
      bForceMonospaced=false
      bMaterialIsStencil=false
      fontMaterial=null
      fontObject=null
      letterSpacing=0
      monospacedWidth=1
      size=9
      skewAmount=0
      typefaceFontName=Regular
      outlineSettings {
       bApplyOutlineToDropShadows=false
       bMiteredCorners=false
       bSeparateFillAlpha=false
       outlineMaterial=null
       outlineSize=0
       outlineColor=<rgba 0 0 0 1>
      }
     }
     hintColorAndOpacity {colorUseRule=UseColor_Foreground, specifiedColor=<rgba 1 0 1 1>}
     selectedBackgroundColor {colorUseRule=UseColor_Foreground, specifiedColor=<rgba 1 0 1 1>}
    }
   }
   slot {class=VerticalBoxSlot, properties {padding=<ltrb 0 0 0 8>}}
  }
  BTN_CopyToClipboard {
   widgetClassPath=&Button
   parent=OuterVBox
   widget=BTN_CopyToClipboard
   properties {
    toolTipText=Copy the list of Generated LODs to the Clipboard for Pasting in a Spreadsheet or other Document
    widgetStyle {
     clickedSlateSound {resourceObject=null}
     disabled {
      bIsDynamicallyLoaded=false
      drawAs=RoundedBox
      imageType=NoImage
      mirroring=NoMirror
      resourceName=None
      resourceObject=null
      tiling=NoTile
      imageSize=<0 0>
      margin=<ltrb 0 0 0 0>
      outlineSettings {
       bUseBrushTransparency=false
       roundingType=FixedRadius
       width=1
       color {colorUseRule=UseColor_ColorTable, specifiedColor=<rgba 0 0 0 0>}
       cornerRadii=<4 4 4 4>
      }
      tintColor {colorUseRule=UseColor_Specified, specifiedColor=<rgba 0.0395 0.0395 0.0395 1>}
      uVRegion {bIsValid=false, max=<0 0>, min=<0 0>}
     }
     disabledForeground {colorUseRule=UseColor_ColorTable, specifiedColor=<rgba 0 0 0 0>}
     hovered {
      bIsDynamicallyLoaded=false
      drawAs=RoundedBox
      imageType=NoImage
      mirroring=NoMirror
      resourceName=None
      resourceObject=null
      tiling=NoTile
      imageSize=<32 32>
      margin=<ltrb 0 0 0 0>
      outlineSettings {
       bUseBrushTransparency=false
       roundingType=FixedRadius
       width=1
       color {colorUseRule=UseColor_ColorTable, specifiedColor=<rgba 0 0 0 0>}
       cornerRadii=<4 4 4 4>
      }
      tintColor {colorUseRule=UseColor_Specified, specifiedColor=<rgba 0.0953 0.0953 0.0953 1>}
      uVRegion {bIsValid=false, max=<0 0>, min=<0 0>}
     }
     hoveredForeground {colorUseRule=UseColor_ColorTable, specifiedColor=<rgba 0 0 0 0>}
     hoveredSlateSound {resourceObject=null}
     normal {
      bIsDynamicallyLoaded=false
      drawAs=RoundedBox
      imageType=NoImage
      mirroring=NoMirror
      resourceName=None
      resourceObject=null
      tiling=NoTile
      imageSize=<32 32>
      margin=<ltrb 0 0 0 0>
      outlineSettings {
       bUseBrushTransparency=false
       roundingType=FixedRadius
       width=1
       color {colorUseRule=UseColor_ColorTable, specifiedColor=<rgba 0 0 0 0>}
       cornerRadii=<4 4 4 4>
      }
      tintColor {colorUseRule=UseColor_Specified, specifiedColor=<rgba 0.0395 0.0395 0.0395 1>}
      uVRegion {bIsValid=false, max=<0 0>, min=<0 0>}
     }
     normalForeground {colorUseRule=UseColor_ColorTable, specifiedColor=<rgba 0 0 0 0>}
     normalPadding=<ltrb 12 1.5 12 1.5>
     pressed {
      bIsDynamicallyLoaded=false
      drawAs=RoundedBox
      imageType=NoImage
      mirroring=NoMirror
      resourceName=None
      resourceObject=null
      tiling=NoTile
      imageSize=<32 32>
      margin=<ltrb 0 0 0 0>
      outlineSettings {
       bUseBrushTransparency=false
       roundingType=FixedRadius
       width=1
       color {colorUseRule=UseColor_ColorTable, specifiedColor=<rgba 0 0 0 0>}
       cornerRadii=<4 4 4 4>
      }
      tintColor {colorUseRule=UseColor_Specified, specifiedColor=<rgba 0.0284 0.0284 0.0284 1>}
      uVRegion {bIsValid=false, max=<0 0>, min=<0 0>}
     }
     pressedForeground {colorUseRule=UseColor_ColorTable, specifiedColor=<rgba 0 0 0 0>}
     pressedPadding=<ltrb 12 2.5 12 0.5>
     pressedSlateSound {resourceObject=null}
    }
   }
   slot {class=VerticalBoxSlot, properties {horizontalAlignment=HAlign_Center, padding=<ltrb 0 0 0 10>}}
  }
  TXT_CopyToClipboardLabel {
   widgetClassPath=&TextBlock
   parent=BTN_CopyToClipboard
   widget=TXT_CopyToClipboardLabel
   properties {
    text=Copy to Clipboard
    font {
     bForceMonospaced=false
     bMaterialIsStencil=false
     fontMaterial=null
     fontObject=&Roboto
     letterSpacing=0
     monospacedWidth=1
     size=13.5
     skewAmount=0
     typefaceFontName=Bold
     outlineSettings {
      bApplyOutlineToDropShadows=false
      bMiteredCorners=false
      bSeparateFillAlpha=false
      outlineMaterial=null
      outlineSize=0
      outlineColor=<rgba 0 0 0 1>
     }
    }
   }
   slot {class=ButtonSlot}
  }
  BTN_ExportToCSV {
   widgetClassPath=&Button
   parent=OuterVBox
   widget=BTN_ExportToCSV
   properties {
    toolTipText=Exports the Generated LODs to a CSV File
    widgetStyle {
     clickedSlateSound {resourceObject=null}
     disabled {
      bIsDynamicallyLoaded=false
      drawAs=RoundedBox
      imageType=NoImage
      mirroring=NoMirror
      resourceName=None
      resourceObject=null
      tiling=NoTile
      imageSize=<0 0>
      margin=<ltrb 0 0 0 0>
      outlineSettings {
       bUseBrushTransparency=false
       roundingType=FixedRadius
       width=1
       color {colorUseRule=UseColor_ColorTable, specifiedColor=<rgba 0 0 0 0>}
       cornerRadii=<4 4 4 4>
      }
      tintColor {colorUseRule=UseColor_Specified, specifiedColor=<rgba 0.0395 0.0395 0.0395 1>}
      uVRegion {bIsValid=false, max=<0 0>, min=<0 0>}
     }
     disabledForeground {colorUseRule=UseColor_ColorTable, specifiedColor=<rgba 0 0 0 0>}
     hovered {
      bIsDynamicallyLoaded=false
      drawAs=RoundedBox
      imageType=NoImage
      mirroring=NoMirror
      resourceName=None
      resourceObject=null
      tiling=NoTile
      imageSize=<32 32>
      margin=<ltrb 0 0 0 0>
      outlineSettings {
       bUseBrushTransparency=false
       roundingType=FixedRadius
       width=1
       color {colorUseRule=UseColor_ColorTable, specifiedColor=<rgba 0 0 0 0>}
       cornerRadii=<4 4 4 4>
      }
      tintColor {colorUseRule=UseColor_Specified, specifiedColor=<rgba 0.0953 0.0953 0.0953 1>}
      uVRegion {bIsValid=false, max=<0 0>, min=<0 0>}
     }
     hoveredForeground {colorUseRule=UseColor_ColorTable, specifiedColor=<rgba 0 0 0 0>}
     hoveredSlateSound {resourceObject=null}
     normal {
      bIsDynamicallyLoaded=false
      drawAs=RoundedBox
      imageType=NoImage
      mirroring=NoMirror
      resourceName=None
      resourceObject=null
      tiling=NoTile
      imageSize=<32 32>
      margin=<ltrb 0 0 0 0>
      outlineSettings {
       bUseBrushTransparency=false
       roundingType=FixedRadius
       width=1
       color {colorUseRule=UseColor_ColorTable, specifiedColor=<rgba 0 0 0 0>}
       cornerRadii=<4 4 4 4>
      }
      tintColor {colorUseRule=UseColor_Specified, specifiedColor=<rgba 0.0395 0.0395 0.0395 1>}
      uVRegion {bIsValid=false, max=<0 0>, min=<0 0>}
     }
     normalForeground {colorUseRule=UseColor_ColorTable, specifiedColor=<rgba 0 0 0 0>}
     normalPadding=<ltrb 12 1.5 12 1.5>
     pressed {
      bIsDynamicallyLoaded=false
      drawAs=RoundedBox
      imageType=NoImage
      mirroring=NoMirror
      resourceName=None
      resourceObject=null
      tiling=NoTile
      imageSize=<32 32>
      margin=<ltrb 0 0 0 0>
      outlineSettings {
       bUseBrushTransparency=false
       roundingType=FixedRadius
       width=1
       color {colorUseRule=UseColor_ColorTable, specifiedColor=<rgba 0 0 0 0>}
       cornerRadii=<4 4 4 4>
      }
      tintColor {colorUseRule=UseColor_Specified, specifiedColor=<rgba 0.0284 0.0284 0.0284 1>}
      uVRegion {bIsValid=false, max=<0 0>, min=<0 0>}
     }
     pressedForeground {colorUseRule=UseColor_ColorTable, specifiedColor=<rgba 0 0 0 0>}
     pressedPadding=<ltrb 12 2.5 12 0.5>
     pressedSlateSound {resourceObject=null}
    }
   }
   slot {class=VerticalBoxSlot, properties {horizontalAlignment=HAlign_Center}}
  }
  TXT_ExportToCSVLabel {
   widgetClassPath=&TextBlock
   parent=BTN_ExportToCSV
   widget=TXT_ExportToCSVLabel
   properties {
    text=Export to CSV File
    colorAndOpacity {colorUseRule=UseColor_Specified, specifiedColor=<rgba 0.9 0.9 0.9 1>}
    font {
     bForceMonospaced=false
     bMaterialIsStencil=false
     fontMaterial=null
     fontObject=&Roboto
     letterSpacing=0
     monospacedWidth=1
     size=13.5
     skewAmount=0
     typefaceFontName=Bold
     outlineSettings {
      bApplyOutlineToDropShadows=false
      bMiteredCorners=false
      bSeparateFillAlpha=false
      outlineMaterial=null
      outlineSize=0
      outlineColor=<rgba 0 0 0 1>
     }
    }
   }
   slot {class=ButtonSlot}
  }
 ]
}
