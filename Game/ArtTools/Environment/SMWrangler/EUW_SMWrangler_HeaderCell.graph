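# Header cell widget for the SMWrangler environment art tool, showing a bold, centered label on a dark background.
# A root Border wraps a single TextBlock, TXT_Label.
# SetLabel replaces the label text and SetMinWidth sets the TextBlock's minimum desired width, so the owner controls each cell's caption and width.
# The event graph is empty and the widget holds no variables of its own beyond the two named widgets.
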WidgetBlueprint /Game/ArtTools/Environment/SMWrangler/EUW_SMWrangler_HeaderCell
graphs [
 EventGraph {links [], nodes []}
 SetLabel {nodes [
   n1 = SetLabel!() -> (then, InLabel:Text) {is=title K2Node_FunctionEntry}
   > "SetText (Text)"(self=Get TXT_Label() -> (TXT_Label:Text Object Reference), InText=@n1)
  ]}
 SetMinWidth {nodes [
   n1 = SetMinWidth!() -> (then, InWidth:double=~) {is=title K2Node_FunctionEntry}
   > Set MinDesiredWidth!(execute, MinDesiredWidth:float=@n1, self:Text Object Reference=Get TXT_Label()) -> (then, Output_Get:float)
  ]}
]
variables []
widgets {
 info {inheritedWidgetCount=0, parentClass=&UserWidget, rootWidgetClass=&Border, widgetCount=2}
 widgets [
  BorderBG {
   widgetClassPath=&Border
   bIsVariable=true
   widget=BorderBG
   properties {
    background {
     bIsDynamicallyLoaded=false
     drawAs=Box
     imageType=NoImage
     mirroring=NoMirror
     resourceName=None
     resourceObject=null
     tiling=NoTile
     imageSize=<32 32>
     margin=<ltrb 0 0 0 0>
     outlineSettings {
      bUseBrushTransparency=false
      roundingType=HalfHeightRadius
      width=0
      color {colorUseRule=UseColor_Specified, specifiedColor=<rgba 0 0 0 0>}
      cornerRadii=<0 0 0 0>
     }
     tintColor {colorUseRule=UseColor_Specified, specifiedColor=<rgba 0.0417 0.0417 0.0417 1>}
     uVRegion {bIsValid=false, max=<0 0>, min=<0 0>}
    }
    padding=<ltrb 2 3 0 3>
   }
  }
  TXT_Label {
   widgetClassPath=&TextBlock
   bIsVariable=true
   parent=BorderBG
   widget=TXT_Label
   properties {
    justification=Center
    minDesiredWidth=92
    text=Header
    colorAndOpacity {colorUseRule=UseColor_Specified, specifiedColor=<rgba 0.9 0.9 0.9 1>}
    font {
     bForceMonospaced=false
     bMaterialIsStencil=false
     fontMaterial=null
     fontObject=&Roboto
     letterSpacing=0
     monospacedWidth=1
     size=11
     skewAmount=0
     typefaceFontName=Bold
     outlineSettings {
      bApplyOutlineToDropShadows=false
      bMiteredCorners=false
      bSeparateFillAlpha=false
      outlineMaterial=null
      outlineSize=0
      outlineColor=<rgba 0 0 0 1>
     }
    }
   }
   slot {class=BorderSlot, properties {padding=<ltrb 2 2 0 2>}}
  }
 ]
}
