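# Editor utility widget cell for the SMWrangler tool, showing one static mesh LOD's type, vertex count and editable screen size.
# InitCell queries the mesh and StaticMeshEditorSubsystem, labelling LODs Imported or Generated from reduction settings (LOD0 always Imported), or blanking cells for missing LODs.
# Edits mark the owning LODRow's button dirty without writing to the mesh.
# The screen size border turns red at zero, on init and commit.
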WidgetBlueprint /Game/ArtTools/Environment/SMWrangler/EUW_SMWrangler_LODCell
graphs [
 EventGraph {nodes [
   n1 = OnTextChanged_Event_0!() -> (OutputDelegate, then, Text:"Text (by ref)") {is=title K2Node_CustomEvent}
   > SetButtonDirty(self:EUW SMWrangler LODRow Object Reference=Get OwningRow() -> (OwningRow:EUW SMWrangler LODRow Object Reference))
   n2 = OnTextCommitted_Event!() -> (OutputDelegate, then, Text:"Text (by ref)", CommitMethod:ETextCommit Enum=Default) {is=title K2Node_CustomEvent}
   n3 = "To String (Text)"(InText=@n2.Text)
   > n4 = Branch(Condition="Nearly Equal (Float)"(A=String To Float(InString=@n3), B=0, ErrorTolerance=0.0001))
   > Set BrushColor!(
    execute
    BrushColor:Linear Color Structure=(R=0.512,G=0.032,B=0.032,A=1)
    self:Border Object Reference=Get Border_2() -> (Border_2:Border Object Reference)
   ) -> (then, Output_Get:Linear Color Structure)
   Set BrushColor(execute=@n4.else, BrushColor=(R=0.05,G=0.05,B=0.05,A=1), self=Get Border_2())
   Event Construct()
   > Assign On Text Changed!(
    execute:Exec
    self:Text Box Object Reference=Get TXT_ScreenSize() -> (TXT_ScreenSize:Text Box Object Reference)
    Delegate:"Delegate (by ref)"=@n1
   ) -> (then:Exec) {is=title K2Node_AssignDelegate}
   > Assign On Text Committed!(execute:Exec, self:Text Box Object Reference=Get TXT_ScreenSize(), Delegate:"Delegate (by ref)"=@n2) -> (then:Exec) {is=title K2Node_AssignDelegate}
   n5 = "To String (Text)"(InText=@n2.Text)
   n6 = Branch(Condition="Nearly Equal (Float)"(A=String To Float(InString=@n5), B=0, ErrorTolerance=0.0001))
   > Set BrushColor(BrushColor=(R=0.512,G=0.032,B=0.032,A=1), self=Get Border_2())
   Set BrushColor(execute=@n6.else, BrushColor=(R=0.05,G=0.05,B=0.05,A=1), self=Get Border_2())
  ]}
 InitCell {nodes [
   n1 = InitCell!() -> (then, InMesh:Static Mesh Object Reference, InLODIndex:int, InOwningRow:EUW SMWrangler LODRow Object Reference) {is=title K2Node_FunctionEntry}
   n2 = Get StaticMeshEditorSubsystem()
   > n3 = GetLodReductionSettings!(execute, self:Static Mesh Editor Subsystem Object Reference=@n2, StaticMesh:Static Mesh Object Reference=@n1, LodIndex:int=@n1.InLODIndex) -> (then, OutReductionOptions:Mesh Reduction Settings Structure)
   > Set OwningMesh(OwningMesh:Static Mesh Object Reference=@n1) -> (Output_Get:Static Mesh Object Reference)
   > Set LODIndex(LODIndex:int=@n1.InLODIndex) -> (Output_Get:int=0)
   > Set OwningRow(OwningRow:EUW SMWrangler LODRow Object Reference=@n1.InOwningRow) -> (Output_Get:EUW SMWrangler LODRow Object Reference)
   > n4 = Branch(Condition="integer < integer"(A=@n1.InLODIndex, B=GetNumLODs(self=@n1)))
   n5 = MakeLiteralString(Value=Imported)
   n6 = Break Mesh Reduction Settings!(MeshReductionSettings:"Mesh Reduction Settings Structure (by ref)"=@n3) -> (
    PercentTriangles:float=~
    PercentVertices:float=~
    MaxDeviation:float=~
    PixelError:float=~
    WeldingThreshold:float=~
    HardAngleThreshold:float=~
    BaseLODModel:int
    SilhouetteImportance:EMeshFeatureImportance Enum
    TextureImportance:EMeshFeatureImportance Enum
    ShadingImportance:EMeshFeatureImportance Enum
    bRecalculateNormals:bool=~
    bGenerateUniqueLightmapUVs:bool=~
    bKeepSymmetry:bool=~
    bVisibilityAided:bool=~
    bCullOccluded:bool=~
    TerminationCriterion:EStaticMeshReductionTerimationCriterion Enum
    VisibilityAggressiveness:EMeshFeatureImportance Enum
    VertexColorImportance:EMeshFeatureImportance Enum
   ) {is=title K2Node_BreakStruct}
   n7 = "float < float"(A=BreakVector(InVec="To Vector (Float)"!(InDouble:double=@n6) -> (ReturnValue:Vector)), B=Make Literal Float(Value=1))
   n8 = Select(Option 0:String=MakeLiteralString(Value=Imported), Option 1:String=MakeLiteralString(Value=Generated), Index=@n7) -> (ReturnValue:String)
   n9 = "Not Equal (Integer)"(A=@n1.InLODIndex, B=MakeLiteralInt!(Value:int=0) -> (ReturnValue:int=0))
   > "SetText (Text)"(self=Get TXT_Type() -> (TXT_Type:Text Object Reference), InText="To Text (String)"(InString=Select(Option 0=@n5, Option 1=@n8, Index=@n9)))
   > SetTypeTextColor!(execute, self=Self-Reference(), bIsGenerated:bool=@n7) -> (then)
   > "SetText (Text)"(
    self=Get TXT_Verts() -> (TXT_Verts:Text Object Reference)
    InText="To Text (Integer)"(Value=GetNumVertices!(self:Static Mesh Object Reference=@n1, LODIndex:int=@n1.InLODIndex) -> (ReturnValue:int=0))
   )
   n10 = "Get (a copy)"(
    Array:"Array of Float (single-precision)s"=GetLodScreenSizes!(self:Static Mesh Editor Subsystem Object Reference=@n2, StaticMesh:Static Mesh Object Reference=@n1) -> (ReturnValue:"Array of Float (single-precision)s")
    Dimension 1=@n1.InLODIndex
   ) -> (Output:float)
   > n11 = Branch(Condition="Nearly Equal (Float)"(A=@n10, B=0, ErrorTolerance=0.0001))
   > n12 = SetBrushColor(self=Get Border_2(), InBrushColor=(R=0.512,G=0.032,B=0.032,A=1))
   "SetText (Text Box)"!(
    execute=[@n12.then, @n13.then]
    self:Text Box Object Reference=Get TXT_ScreenSize()
    InText:Text="To Text (Float)"(Value=@n10, bUseGrouping=false, MinimumIntegralDigits=1, MaximumIntegralDigits=6, MinimumFractionalDigits=2, MaximumFractionalDigits=2)
   ) -> (then)
   n13 = SetBrushColor(execute=@n11.else, self=Get Border_2(), InBrushColor=(R=0.05,G=0.05,B=0.05,A=1))
   "SetText (Text)"(execute=@n4.else, self=Get TXT_Type(), InText="To Text (String)"())
   > "SetText (Text)"(self=Get TXT_Verts(), InText="To Text (String)"())
   > SetVisibility(self=Get TXT_ScreenSize(), InVisibility=Hidden)
  ]}
 SetTypeTextColor {nodes [
   n1 = SetTypeTextColor!() -> (then, bIsGenerated:bool=~) {is=title K2Node_FunctionEntry}
   > n2 = Branch(Condition=@n1)
   n3 = Get TXT_Type()
   > Set ColorAndOpacity!(
    execute
    ColorAndOpacity:Slate Color Structure=Make Slate Color(SpecifiedColor=Make Linear Color!(R:float=0.75, G:float=0.02, B:float=0.02, A:float=1) -> (LinearColor:Linear Color Structure) {is=title K2Node_MakeStruct})
    self:Text Object Reference=@n3
   ) -> (then, Output_Get:Slate Color Structure)
   n4 = MakeSlateFontInfo!(
    FontObject:Object Reference=&Roboto
    FontMaterial:Object Reference
    OutlineSettings:Font Outline Settings Structure
    TypefaceFontName:Name=Bold
    Size:float=11
    LetterSpacing:int=0
    SkewAmount:float=0
    bForceMonospaced:bool=false
    bMaterialIsStencil:bool=false
    MonospacedWidth:float=1
   ) -> (ReturnValue:Slate Font Info Structure)
   > Set Font!(execute, Font:Slate Font Info Structure=@n4, self:Text Object Reference=@n3) -> (then, Output_Get:Slate Font Info Structure)
   Set ColorAndOpacity(execute=@n2.else, ColorAndOpacity=Make Slate Color(SpecifiedColor=Make Linear Color(R=0.4, G=0.4, B=0.4, A=1)), self=@n3)
   > Set Font(Font=@n4, self=@n3)
  ]}
]
variables [
 LODIndex {type=integer}
 OwningMesh {type=StaticMesh, default=null}
 OwningRow {type=EUW_SMWrangler_LODRow_C, default=null}
]
widgets {
 info {inheritedWidgetCount=0, parentClass=&EditorUtilityWidget, rootWidgetClass=&HorizontalBox, widgetCount=7}
 widgets [
  CellRoot {widgetClassPath=&HorizontalBox, bIsVariable=true, widget=CellRoot}
  Border_0 {
   widgetClassPath=&Border
   bIsVariable=true
   parent=CellRoot
   widget=Border_0
   properties {
    verticalAlignment=VAlign_Center
    background {
     bIsDynamicallyLoaded=false
     drawAs=Image
     imageType=FullColor
     mirroring=NoMirror
     resourceName=None
     resourceObject=null
     tiling=NoTile
     imageSize=<0 0>
     margin=<ltrb 0 0 0 0>
     outlineSettings {
      bUseBrushTransparency=false
      roundingType=HalfHeightRadius
      width=0
      color {colorUseRule=UseColor_Specified, specifiedColor=<rgba 0 0 0 0>}
      cornerRadii=<0 0 0 0>
     }
     tintColor {colorUseRule=UseColor_Specified, specifiedColor=<rgba 1 1 1 1>}
     uVRegion {bIsValid=false, max=<0 0>, min=<0 0>}
    }
    brushColor=<rgba 0.0417 0.0417 0.0417 1>
    padding=<ltrb 4 3 4 3>
   }
   slot {class=HorizontalBoxSlot, properties {verticalAlignment=VAlign_Center, padding=<ltrb 4 0 0 0>}}
  }
  TXT_Type {
   widgetClassPath=&TextBlock
   bIsVariable=true
   parent=Border_0
   widget=TXT_Type
   properties {
    justification=Center
    minDesiredWidth=92
    text=Imported
    font {
     bForceMonospaced=false
     bMaterialIsStencil=false
     fontMaterial=null
     fontObject=&Roboto
     letterSpacing=0
     monospacedWidth=1
     size=11
     skewAmount=0
     typefaceFontName=Regular
     outlineSettings {
      bApplyOutlineToDropShadows=false
      bMiteredCorners=false
      bSeparateFillAlpha=false
      outlineMaterial=null
      outlineSize=0
      outlineColor=<rgba 0 0 0 1>
     }
    }
   }
   slot {class=BorderSlot, properties {verticalAlignment=VAlign_Center}}
  }
  Border_1 {
   widgetClassPath=&Border
   bIsVariable=true
   parent=CellRoot
   widget=Border_1
   properties {
    verticalAlignment=VAlign_Center
    background {
     bIsDynamicallyLoaded=false
     drawAs=Image
     imageType=FullColor
     mirroring=NoMirror
     resourceName=None
     resourceObject=null
     tiling=NoTile
     imageSize=<0 0>
     margin=<ltrb 0 0 0 0>
     outlineSettings {
      bUseBrushTransparency=false
      roundingType=HalfHeightRadius
      width=0
      color {colorUseRule=UseColor_Specified, specifiedColor=<rgba 0 0 0 0>}
      cornerRadii=<0 0 0 0>
     }
     tintColor {colorUseRule=UseColor_Specified, specifiedColor=<rgba 1 1 1 1>}
     uVRegion {bIsValid=false, max=<0 0>, min=<0 0>}
    }
    brushColor=<rgba 0.0417 0.0417 0.0417 1>
    padding=<ltrb 4 3 4 3>
   }
   slot {class=HorizontalBoxSlot, properties {verticalAlignment=VAlign_Center, padding=<ltrb 2 0 0 0>}}
  }
  TXT_Verts {
   widgetClassPath=&TextBlock
   bIsVariable=true
   parent=Border_1
   widget=TXT_Verts
   properties {
    justification=Center
    minDesiredWidth=72
    text=0
    font {
     bForceMonospaced=false
     bMaterialIsStencil=false
     fontMaterial=null
     fontObject=&Roboto
     letterSpacing=0
     monospacedWidth=1
     size=11
     skewAmount=0
     typefaceFontName=Bold
     outlineSettings {
      bApplyOutlineToDropShadows=false
      bMiteredCorners=false
      bSeparateFillAlpha=false
      outlineMaterial=null
      outlineSize=0
      outlineColor=<rgba 0 0 0 1>
     }
    }
   }
   slot {class=BorderSlot, properties {verticalAlignment=VAlign_Center}}
  }
  Border_2 {
   widgetClassPath=&Border
   bIsVariable=true
   parent=CellRoot
   widget=Border_2
   properties {
    toolTipText="Editable Screen Size field\r\nRed = Screen Size = 0"
    verticalAlignment=VAlign_Center
    brushColor=<rgba 0.04 0.04 0.04 1>
    padding=<ltrb 1 1 1 1>
   }
   slot {class=HorizontalBoxSlot, properties {verticalAlignment=VAlign_Center, padding=<ltrb 2 0 0 0>}}
  }
  TXT_ScreenSize {
   widgetClassPath=&EditableTextBox
   bIsVariable=true
   parent=Border_2
   widget=TXT_ScreenSize
   properties {
    justification=Center
    minimumDesiredWidth=108
    text=0.000
    widgetStyle {
     backgroundColor {colorUseRule=UseColor_Specified, specifiedColor=<rgba 1 1 1 1>}
     backgroundImageFocused {
      bIsDynamicallyLoaded=false
      drawAs=Image
      imageType=NoImage
      mirroring=NoMirror
      resourceName=../../../Engine/Content/Slate/Common/TextBox_Hovered.png
      resourceObject=null
      tiling=NoTile
      imageSize=<0 0>
      margin=<ltrb 0 0 0 0>
      outlineSettings {
       bUseBrushTransparency=false
       roundingType=HalfHeightRadius
       width=0
       color {colorUseRule=UseColor_Specified, specifiedColor=<rgba 0 0 0 0>}
       cornerRadii=<0 0 0 0>
      }
      tintColor {colorUseRule=UseColor_Specified, specifiedColor=<rgba 0.0417 0.0417 0.0417 1>}
      uVRegion {bIsValid=false, max=<0 0>, min=<0 0>}
     }
     backgroundImageHovered {
      bIsDynamicallyLoaded=false
      drawAs=Image
      imageType=NoImage
      mirroring=NoMirror
      resourceName=../../../Engine/Content/Slate/Common/TextBox_Hovered.png
      resourceObject=null
      tiling=NoTile
      imageSize=<0 0>
      margin=<ltrb 0 0 0 0>
      outlineSettings {
       bUseBrushTransparency=false
       roundingType=HalfHeightRadius
       width=0
       color {colorUseRule=UseColor_Specified, specifiedColor=<rgba 0 0 0 0>}
       cornerRadii=<0 0 0 0>
      }
      tintColor {colorUseRule=UseColor_Specified, specifiedColor=<rgba 0.015 0.015 0.015 1>}
      uVRegion {bIsValid=false, max=<0 0>, min=<0 0>}
     }
     backgroundImageNormal {
      bIsDynamicallyLoaded=false
      drawAs=Image
      imageType=NoImage
      mirroring=NoMirror
      resourceName=../../../Engine/Content/Slate/Common/TextBox.png
      resourceObject=null
      tiling=NoTile
      imageSize=<0 0>
      margin=<ltrb 0 0 0 0>
      outlineSettings {
       bUseBrushTransparency=false
       roundingType=HalfHeightRadius
       width=0
       color {colorUseRule=UseColor_Specified, specifiedColor=<rgba 0 0 0 0>}
       cornerRadii=<0 0 0 0>
      }
      tintColor {colorUseRule=UseColor_Specified, specifiedColor=<rgba 0.015 0.015 0.015 1>}
      uVRegion {bIsValid=false, max=<0 0>, min=<0 0>}
     }
     backgroundImageReadOnly {
      bIsDynamicallyLoaded=false
      drawAs=Image
      imageType=NoImage
      mirroring=NoMirror
      resourceName=../../../Engine/Content/Slate/Common/TextBox_ReadOnly.png
      resourceObject=null
      tiling=NoTile
      imageSize=<0 0>
      margin=<ltrb 0 0 0 0>
      outlineSettings {
       bUseBrushTransparency=false
       roundingType=HalfHeightRadius
       width=0
       color {colorUseRule=UseColor_Specified, specifiedColor=<rgba 0 0 0 0>}
       cornerRadii=<0 0 0 0>
      }
      tintColor {colorUseRule=UseColor_Specified, specifiedColor=<rgba 0.015 0.015 0.015 1>}
      uVRegion {bIsValid=false, max=<0 0>, min=<0 0>}
     }
     focusedForegroundColor {colorUseRule=UseColor_Foreground, specifiedColor=<rgba 1 0 1 1>}
     foregroundColor {colorUseRule=UseColor_Specified, specifiedColor=<rgba 0.9 0.9 0.9 1>}
     hScrollBarPadding=<ltrb 0 0 0 0>
     padding=<ltrb 0 0 0 0>
     readOnlyForegroundColor {colorUseRule=UseColor_Foreground, specifiedColor=<rgba 1 0 1 1>}
     scrollBarStyle {
      thickness=16
      draggedThumbImage {
       bIsDynamicallyLoaded=false
       drawAs=Box
       imageType=FullColor
       mirroring=NoMirror
       resourceName=../../../Engine/Content/Slate/Common/Scrollbar_Thumb.png
       resourceObject=null
       tiling=NoTile
       imageSize=<0 0>
       margin=<ltrb 0.25 0.25 0.25 0.25>
       outlineSettings {
        bUseBrushTransparency=false
        roundingType=HalfHeightRadius
        width=0
        color {colorUseRule=UseColor_Specified, specifiedColor=<rgba 0 0 0 0>}
        cornerRadii=<0 0 0 0>
       }
       tintColor {colorUseRule=UseColor_Specified, specifiedColor=<rgba 1 1 1 1>}
       uVRegion {bIsValid=false, max=<0 0>, min=<0 0>}
      }
      horizontalBackgroundImage {
       bIsDynamicallyLoaded=false
       drawAs=NoDrawType
       imageType=NoImage
       mirroring=NoMirror
       resourceName=None
       resourceObject=null
       tiling=NoTile
       imageSize=<0 0>
       margin=<ltrb 0 0 0 0>
       outlineSettings {
        bUseBrushTransparency=false
        roundingType=HalfHeightRadius
        width=0
        color {colorUseRule=UseColor_Specified, specifiedColor=<rgba 0 0 0 0>}
        cornerRadii=<0 0 0 0>
       }
       tintColor {colorUseRule=UseColor_Specified, specifiedColor=<rgba 1 1 1 1>}
       uVRegion {bIsValid=false, max=<0 0>, min=<0 0>}
      }
      horizontalBottomSlotImage {
       bIsDynamicallyLoaded=false
       drawAs=Image
       imageType=FullColor
       mirroring=NoMirror
       resourceName=../../../Engine/Content/Slate/Common/Scrollbar_Background_Horizontal.png
       resourceObject=null
       tiling=NoTile
       imageSize=<8 8>
       margin=<ltrb 0 0 0 0>
       outlineSettings {
        bUseBrushTransparency=false
        roundingType=HalfHeightRadius
        width=0
        color {colorUseRule=UseColor_Specified, specifiedColor=<rgba 0 0 0 0>}
        cornerRadii=<0 0 0 0>
       }
       tintColor {colorUseRule=UseColor_Specified, specifiedColor=<rgba 1 1 1 1>}
       uVRegion {bIsValid=false, max=<0 0>, min=<0 0>}
      }
      horizontalTopSlotImage {
       bIsDynamicallyLoaded=false
       drawAs=Image
       imageType=FullColor
       mirroring=NoMirror
       resourceName=../../../Engine/Content/Slate/Common/Scrollbar_Background_Horizontal.png
       resourceObject=null
       tiling=NoTile
       imageSize=<8 8>
       margin=<ltrb 0 0 0 0>
       outlineSettings {
        bUseBrushTransparency=false
        roundingType=HalfHeightRadius
        width=0
        color {colorUseRule=UseColor_Specified, specifiedColor=<rgba 0 0 0 0>}
        cornerRadii=<0 0 0 0>
       }
       tintColor {colorUseRule=UseColor_Specified, specifiedColor=<rgba 1 1 1 1>}
       uVRegion {bIsValid=false, max=<0 0>, min=<0 0>}
      }
      hoveredThumbImage {
       bIsDynamicallyLoaded=false
       drawAs=Box
       imageType=FullColor
       mirroring=NoMirror
       resourceName=../../../Engine/Content/Slate/Common/Scrollbar_Thumb.png
       resourceObject=null
       tiling=NoTile
       imageSize=<0 0>
       margin=<ltrb 0.25 0.25 0.25 0.25>
       outlineSettings {
        bUseBrushTransparency=false
        roundingType=HalfHeightRadius
        width=0
        color {colorUseRule=UseColor_Specified, specifiedColor=<rgba 0 0 0 0>}
        cornerRadii=<0 0 0 0>
       }
       tintColor {colorUseRule=UseColor_Specified, specifiedColor=<rgba 1 1 1 1>}
       uVRegion {bIsValid=false, max=<0 0>, min=<0 0>}
      }
      normalThumbImage {
       bIsDynamicallyLoaded=false
       drawAs=Box
       imageType=FullColor
       mirroring=NoMirror
       resourceName=../../../Engine/Content/Slate/Common/Scrollbar_Thumb.png
       resourceObject=null
       tiling=NoTile
       imageSize=<0 0>
       margin=<ltrb 0.25 0.25 0.25 0.25>
       outlineSettings {
        bUseBrushTransparency=false
        roundingType=HalfHeightRadius
        width=0
        color {colorUseRule=UseColor_Specified, specifiedColor=<rgba 0 0 0 0>}
        cornerRadii=<0 0 0 0>
       }
       tintColor {colorUseRule=UseColor_Specified, specifiedColor=<rgba 1 1 1 1>}
       uVRegion {bIsValid=false, max=<0 0>, min=<0 0>}
      }
      verticalBackgroundImage {
       bIsDynamicallyLoaded=false
       drawAs=NoDrawType
       imageType=NoImage
       mirroring=NoMirror
       resourceName=None
       resourceObject=null
       tiling=NoTile
       imageSize=<0 0>
       margin=<ltrb 0 0 0 0>
       outlineSettings {
        bUseBrushTransparency=false
        roundingType=HalfHeightRadius
        width=0
        color {colorUseRule=UseColor_Specified, specifiedColor=<rgba 0 0 0 0>}
        cornerRadii=<0 0 0 0>
       }
       tintColor {colorUseRule=UseColor_Specified, specifiedColor=<rgba 1 1 1 1>}
       uVRegion {bIsValid=false, max=<0 0>, min=<0 0>}
      }
      verticalBottomSlotImage {
       bIsDynamicallyLoaded=false
       drawAs=Image
       imageType=FullColor
       mirroring=NoMirror
       resourceName=../../../Engine/Content/Slate/Common/Scrollbar_Background_Vertical.png
       resourceObject=null
       tiling=NoTile
       imageSize=<8 8>
       margin=<ltrb 0 0 0 0>
       outlineSettings {
        bUseBrushTransparency=false
        roundingType=HalfHeightRadius
        width=0
        color {colorUseRule=UseColor_Specified, specifiedColor=<rgba 0 0 0 0>}
        cornerRadii=<0 0 0 0>
       }
       tintColor {colorUseRule=UseColor_Specified, specifiedColor=<rgba 1 1 1 1>}
       uVRegion {bIsValid=false, max=<0 0>, min=<0 0>}
      }
      verticalTopSlotImage {
       bIsDynamicallyLoaded=false
       drawAs=Image
       imageType=FullColor
       mirroring=NoMirror
       resourceName=../../../Engine/Content/Slate/Common/Scrollbar_Background_Vertical.png
       resourceObject=null
       tiling=NoTile
       imageSize=<8 8>
       margin=<ltrb 0 0 0 0>
       outlineSettings {
        bUseBrushTransparency=false
        roundingType=HalfHeightRadius
        width=0
        color {colorUseRule=UseColor_Specified, specifiedColor=<rgba 0 0 0 0>}
        cornerRadii=<0 0 0 0>
       }
       tintColor {colorUseRule=UseColor_Specified, specifiedColor=<rgba 1 1 1 1>}
       uVRegion {bIsValid=false, max=<0 0>, min=<0 0>}
      }
     }
     textStyle {
      overflowPolicy=Clip
      transformPolicy=None
      colorAndOpacity {colorUseRule=UseColor_Specified, specifiedColor=<rgba 0.9 0.9 0.9 1>}
      font {
       bForceMonospaced=false
       bMaterialIsStencil=false
       fontMaterial=null
       fontObject=&Roboto
       letterSpacing=0
       monospacedWidth=1
       size=11
       skewAmount=0
       typefaceFontName=Regular
       outlineSettings {
        bApplyOutlineToDropShadows=false
        bMiteredCorners=false
        bSeparateFillAlpha=false
        outlineMaterial=null
        outlineSize=0
        outlineColor=<rgba 0 0 0 1>
       }
      }
      highlightColor {colorUseRule=UseColor_Specified, specifiedColor=<rgba 0.02 0.3 0 1>}
      highlightShape {
       bIsDynamicallyLoaded=false
       drawAs=Box
       imageType=FullColor
       mirroring=NoMirror
       resourceName=../../../Engine/Content/Slate/Common/TextBlockHighlightShape.png
       resourceObject=null
       tiling=NoTile
       imageSize=<0 0>
       margin=<ltrb 0.375 0.375 0.375 0.375>
       outlineSettings {
        bUseBrushTransparency=false
        roundingType=HalfHeightRadius
        width=0
        color {colorUseRule=UseColor_Specified, specifiedColor=<rgba 0 0 0 0>}
        cornerRadii=<0 0 0 0>
       }
       tintColor {colorUseRule=UseColor_Specified, specifiedColor=<rgba 1 1 1 1>}
       uVRegion {bIsValid=false, max=<0 0>, min=<0 0>}
      }
      selectedBackgroundColor {colorUseRule=UseColor_Foreground, specifiedColor=<rgba 1 0 1 1>}
      shadowColorAndOpacity=<rgba 0 0 0 1>
      shadowOffset=<0 0>
      strikeBrush {
       bIsDynamicallyLoaded=false
       drawAs=Image
       imageType=NoImage
       mirroring=NoMirror
       resourceName=None
       resourceObject=null
       tiling=NoTile
       imageSize=<32 32>
       margin=<ltrb 0 0 0 0>
       outlineSettings {
        bUseBrushTransparency=false
        roundingType=HalfHeightRadius
        width=0
        color {colorUseRule=UseColor_Specified, specifiedColor=<rgba 0 0 0 0>}
        cornerRadii=<0 0 0 0>
       }
       tintColor {colorUseRule=UseColor_Specified, specifiedColor=<rgba 1 1 1 1>}
       uVRegion {bIsValid=false, max=<0 0>, min=<0 0>}
      }
      underlineBrush {
       bIsDynamicallyLoaded=false
       drawAs=Image
       imageType=NoImage
       mirroring=NoMirror
       resourceName=None
       resourceObject=null
       tiling=NoTile
       imageSize=<32 32>
       margin=<ltrb 0 0 0 0>
       outlineSettings {
        bUseBrushTransparency=false
        roundingType=HalfHeightRadius
        width=0
        color {colorUseRule=UseColor_Specified, specifiedColor=<rgba 0 0 0 0>}
        cornerRadii=<0 0 0 0>
       }
       tintColor {colorUseRule=UseColor_Specified, specifiedColor=<rgba 1 1 1 1>}
       uVRegion {bIsValid=false, max=<0 0>, min=<0 0>}
      }
     }
     vScrollBarPadding=<ltrb 0 0 0 0>
    }
   }
   slot {class=BorderSlot, properties {verticalAlignment=VAlign_Center, padding=<ltrb 1 3 1 3>}}
  }
 ]
}
