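# Editor utility widget row for the SMWrangler tool, showing one static mesh's per-LOD screen sizes with open-in-editor and save buttons.
# InitRow builds an EUW_SMWrangler_LODCell for each LOD column.
# SaveMesh checks out the asset, gathers screen sizes from cell text boxes, applies them through StaticMeshEditorSubsystem and saves.
# Dirty state toggles "*" label prefixes and notifies the owning Wrangler, suppressing dirty notifications during initialization.
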WidgetBlueprint /Game/ArtTools/Environment/SMWrangler/EUW_SMWrangler_LODRow
graphs [
 EventGraph {nodes [
   "On Clicked (BTN_MeshName)"() {is=title K2Node_ComponentBoundEvent}
   > OpenEditorForAssets!(
    execute
    self:Asset Editor Subsystem Object Reference=Get AssetEditorSubsystem()
    Assets:Array of Object References=Make Array("[0]":Object Reference=Get OwningMesh() -> (OwningMesh:Static Mesh Object Reference)) -> (Array:Array of Object References)
    OpenedMethod:EAssetTypeActivationOpenedMethod Enum=Edit
   ) -> (then, ReturnValue:bool)
   Event Construct()
   > Assign On Clicked!(
    execute:Exec
    self:Button Object Reference=Get BTN_Save() -> (BTN_Save:Button Object Reference)
    Delegate:"Delegate (by ref)"=Create Event(self=Self-Reference())
   ) -> (then:Exec) {is=title K2Node_AssignDelegate}
   > SetButtonClean()
  ]}
 InitRow {nodes [
   n1 = InitRow!() -> (then, InMesh:Static Mesh Object Reference, InMaxLODCount:int, InOwningWidget:EUW SMWrangler Object Reference) {is=title K2Node_FunctionEntry}
   > Set OwningWidget(OwningWidget:EUW SMWrangler Object Reference=@n1.InOwningWidget) -> (Output_Get:EUW SMWrangler Object Reference)
   > Set IsInitializing(IsInitializing:bool=true) -> (Output_Get:bool)
   > Set OwningMesh(OwningMesh:Static Mesh Object Reference=@n1) -> (Output_Get:Static Mesh Object Reference)
   > Set MeshName(MeshName:String=GetObjectName(Object=@n1)) -> (Output_Get:String)
   > "SetText (Text)"(self=Get TXT_MeshName() -> (TXT_MeshName:Text Object Reference), InText="To Text (String)"(InString=Get MeshName() -> (MeshName:String)))
   > ClearChildren(self=Get HB_LODColumns() -> (HB_LODColumns:Horizontal Box Object Reference))
   > n2 = For Loop(LastIndex=int - int(A=@n1.InMaxLODCount, B=MakeLiteralInt!(Value:int=1) -> (ReturnValue:int=0)))
   Set IsInitializing(execute=@n2.Completed, IsInitializing=false)
   n3 = Create Widget!(
    execute:Exec=@n2.LoopBody
    WorldContextObject:Object Reference
    Class:User Widget Class Reference=&/EUW_SMWrangler_LODCell.EUW_SMWrangler_LODCell_C
    OwningPlayer:Player Controller Object Reference
   ) -> (then:Exec, ReturnValue:EUW SMWrangler LODCell Object Reference)
   > n4 = Cast To EUW_SMWrangler_LODCell!(execute, Object=@n3) -> (then, CastFailed, AsEUW SMWrangler LODCell:EUW SMWrangler LODCell Object Reference) {is=title K2Node_DynamicCast}
   > AddChild(self=Get HB_LODColumns(), Content=@n4)
   > InitCell!(
    execute
    self:EUW SMWrangler LODCell Object Reference=@n4
    InMesh:Static Mesh Object Reference=@n1
    InLODIndex:int=@n2
    InOwningRow:EUW SMWrangler LODRow Object Reference=Self-Reference()
   ) -> (then)
  ]}
 SaveMesh {nodes [
   SaveMesh() {is=title K2Node_FunctionEntry}
   n1 = Get OwningMesh()
   > CheckoutLoadedAsset!(execute, AssetToCheckout:Object Reference=@n1) -> (then, ReturnValue:bool)
   n2 = Get StaticMeshEditorSubsystem()
   > Set SaveScreenSizes(SaveScreenSizes:"Array of Float (double-precision)s"=GetLodScreenSizes!(self:Static Mesh Editor Subsystem Object Reference=@n2, StaticMesh:Static Mesh Object Reference=@n1) -> (ReturnValue:"Array of Float (single-precision)s")) -> (Output_Get:"Array of Float (double-precision)s")
   n3 = Get HB_LODColumns()
   > n4 = For Loop(LastIndex=int - int(A=GetChildrenCount!(self:Panel Widget Object Reference=@n3) -> (ReturnValue:int=0), B=MakeLiteralInt(Value=1) -> (ReturnValue=0)))
   SetLodScreenSizes!(
    execute=@n4.Completed
    self:Static Mesh Editor Subsystem Object Reference=@n2
    StaticMesh:Static Mesh Object Reference=@n1
    ScreenSizes:"Array of Float (single-precision)s"=Get SaveScreenSizes() -> (SaveScreenSizes:"Array of Float (double-precision)s")
   ) -> (then, ReturnValue:bool)
   > SaveLoadedAsset!(
    execute
    self:Editor Asset Subsystem Object Reference=Get EditorAssetSubsystem() -> (ReturnValue:Editor Asset Subsystem Object Reference) {is=title K2Node_GetEditorSubsystem}
    AssetToSave=@n1
    bOnlyIfIsDirty=true
   ) -> (then, ReturnValue)
   > SetButtonClean()
   n5 = Cast To EUW_SMWrangler_LODCell(execute=@n4.LoopBody, Object=GetChildAt!(self:Panel Widget Object Reference=@n3, Index:int=@n4) -> (ReturnValue:Widget Object Reference))
   n6 = "GetText (Text Box)"!(self:Text Box Object Reference=Get TXT_ScreenSize!(self:EUW SMWrangler LODCell Object Reference=@n5) -> (TXT_ScreenSize:Text Box Object Reference)) -> (ReturnValue:Text)
   > Set Array Elem!(
    execute
    TargetArray:"Array of Float (double-precision)s"=Get SaveScreenSizes()
    Index:int=Get LODIndex!(self:EUW SMWrangler LODCell Object Reference=@n5) -> (LODIndex:int=0)
    Item:"Float (double-precision) (by ref)"=String To Float(InString="To String (Text)"(InText=@n6))
    bSizeToFit:bool=false
   ) -> (then)
  ]}
 SetButtonClean {nodes [
   SetButtonClean() {is=title K2Node_FunctionEntry}
   > "SetText (Text)"(self=Get TXT_MeshName(), InText="To Text (String)"(InString=Get MeshName()))
   > "SetText (Text)"(self=Get TXT_SaveLabel() -> (TXT_SaveLabel:Text Object Reference), InText=MakeLiteralText(Value="LOCTEXT(\"Save\")"))
   > Branch(Condition=Get IsDirty() -> (IsDirty:bool))
   > Set IsDirty(IsDirty:bool=false) -> (Output_Get:bool)
   n1 = Get OwningWidget() -> (OwningWidget:EUW SMWrangler Object Reference)
   > Is Valid(InputObject=@n1)
   > Set IsSaved(IsSaved:bool=true) -> (Output_Get:bool)
   > NotifyClean(self:EUW SMWrangler Object Reference=@n1)
  ]}
 SetButtonDirty {nodes [
   SetButtonDirty() {is=title K2Node_FunctionEntry}
   n1 = Get MeshName()
   > "SetText (Text)"(self=Get TXT_MeshName(), InText="To Text (String)"(InString=Append(A="* ", B=@n1)))
   > "SetText (Text)"(self=Get TXT_SaveLabel(), InText=MakeLiteralText(Value="LOCTEXT(\"* Save\")"))
   > Branch(Condition=NOT Boolean(A=Get IsDirty()))
   > Set IsDirty(IsDirty=true)
   > Branch(Condition=NOT Boolean(A=Get IsInitializing() -> (IsInitializing:bool)))
   n2 = Get OwningWidget()
   > Is Valid(InputObject=@n2)
   > NotifyDirty!(execute, self:EUW SMWrangler Object Reference=@n2, WasSaved:bool=Get IsSaved() -> (IsSaved:bool)) -> (then)
   > Set IsSaved(IsSaved=false)
  ]}
]
variables [
 IsDirty {type=boolean}
 IsInitializing {type=boolean}
 IsSaved {type=boolean}
 LODIndex {type=integer}
 LodTypeTag {type=string, default=""}
 MeshAssetPath {type=string, default=""}
 MeshName {type=string, default=""}
 OwningMesh {type=StaticMesh, default=null}
 OwningWidget {type=EUW_SMWrangler_C, default=null}
 SaveScreenSizes {type="array<number>", default []}
]
widgets {
 info {inheritedWidgetCount=0, parentClass=&EditorUtilityWidget, rootWidgetClass=&HorizontalBox, widgetCount=8}
 widgets [
  LODRowRoot {widgetClassPath=&HorizontalBox, bIsVariable=true, widget=LODRowRoot}
  Border_0 {
   widgetClassPath=&Border
   bIsVariable=true
   parent=LODRowRoot
   widget=Border_0
   properties {
    toolTipText="Opens up the Mesh in the Static Mesh Editor\r\n* Prefix = The Mesh is Edited with Unsaved changes"
    background {
     bIsDynamicallyLoaded=false
     drawAs=Image
     imageType=FullColor
     mirroring=NoMirror
     resourceName=None
     resourceObject=null
     tiling=NoTile
     imageSize=<0 0>
     margin=<ltrb 0 0 0 0>
     outlineSettings {
      bUseBrushTransparency=false
      roundingType=HalfHeightRadius
      width=0
      color {colorUseRule=UseColor_Specified, specifiedColor=<rgba 0 0 0 0>}
      cornerRadii=<0 0 0 0>
     }
     tintColor {colorUseRule=UseColor_Specified, specifiedColor=<rgba 1 1 1 1>}
     uVRegion {bIsValid=false, max=<0 0>, min=<0 0>}
    }
    brushColor=<rgba 0.0417 0.0417 0.0417 1>
    padding=<ltrb 4 3 4 3>
   }
   slot {class=HorizontalBoxSlot, properties {verticalAlignment=VAlign_Center, padding=<ltrb 2 0 0 0>}}
  }
  BTN_MeshName {
   widgetClassPath=&Button
   parent=Border_0
   widget=BTN_MeshName
   properties {
    isFocusable=false
    widgetStyle {
     clickedSlateSound {resourceObject=null}
     disabled {
      bIsDynamicallyLoaded=false
      drawAs=Image
      imageType=FullColor
      mirroring=NoMirror
      resourceName=None
      resourceObject=null
      tiling=NoTile
      imageSize=<0 0>
      margin=<ltrb 0 0 0 0>
      outlineSettings {
       bUseBrushTransparency=false
       roundingType=HalfHeightRadius
       width=0
       color {colorUseRule=UseColor_Specified, specifiedColor=<rgba 0 0 0 0>}
       cornerRadii=<0 0 0 0>
      }
      tintColor {colorUseRule=UseColor_Specified, specifiedColor=<rgba 0.0417 0.0417 0.0417 1>}
      uVRegion {bIsValid=false, max=<0 0>, min=<0 0>}
     }
     disabledForeground {colorUseRule=UseColor_Specified, specifiedColor=<rgba 0.5271 0.5271 0.5271 1>}
     hovered {
      bIsDynamicallyLoaded=false
      drawAs=Image
      imageType=FullColor
      mirroring=NoMirror
      resourceName=None
      resourceObject=null
      tiling=NoTile
      imageSize=<0 0>
      margin=<ltrb 0 0 0 0>
      outlineSettings {
       bUseBrushTransparency=true
       roundingType=FixedRadius
       width=1
       color {colorUseRule=UseColor_Specified, specifiedColor=<rgba 0.7243 0.7243 0.7243 1>}
       cornerRadii=<4 4 4 4>
      }
      tintColor {colorUseRule=UseColor_Specified, specifiedColor=<rgba 0.0953 0.0953 0.0953 1>}
      uVRegion {bIsValid=false, max=<0 0>, min=<0 0>}
     }
     hoveredForeground {colorUseRule=UseColor_Specified, specifiedColor=<rgba 0 1 1 1>}
     hoveredSlateSound {resourceObject=null}
     normal {
      bIsDynamicallyLoaded=false
      drawAs=Image
      imageType=FullColor
      mirroring=NoMirror
      resourceName=None
      resourceObject=null
      tiling=NoTile
      imageSize=<0 0>
      margin=<ltrb 0 0 0 0>
      outlineSettings {
       bUseBrushTransparency=true
       roundingType=FixedRadius
       width=1
       color {colorUseRule=UseColor_Specified, specifiedColor=<rgba 0.6951 0.6951 0.6951 1>}
       cornerRadii=<4 4 4 4>
      }
      tintColor {colorUseRule=UseColor_Specified, specifiedColor=<rgba 0.0417 0.0417 0.0417 1>}
      uVRegion {bIsValid=false, max=<0 0>, min=<0 0>}
     }
     normalForeground {colorUseRule=UseColor_Specified, specifiedColor=<rgba 0 1 1 1>}
     normalPadding=<ltrb 0 0 0 0>
     pressed {
      bIsDynamicallyLoaded=false
      drawAs=Image
      imageType=FullColor
      mirroring=NoMirror
      resourceName=None
      resourceObject=null
      tiling=NoTile
      imageSize=<0 0>
      margin=<ltrb 0 0 0 0>
      outlineSettings {
       bUseBrushTransparency=true
       roundingType=FixedRadius
       width=1
       color {colorUseRule=UseColor_Specified, specifiedColor=<rgba 0.7243 0.7243 0.7243 1>}
       cornerRadii=<4 4 4 4>
      }
      tintColor {colorUseRule=UseColor_Specified, specifiedColor=<rgba 0.0284 0.0284 0.0284 1>}
      uVRegion {bIsValid=false, max=<0 0>, min=<0 0>}
     }
     pressedForeground {colorUseRule=UseColor_Specified, specifiedColor=<rgba 0 1 1 1>}
     pressedPadding=<ltrb 0 0 0 0>
     pressedSlateSound {resourceObject=null}
    }
   }
   slot {class=BorderSlot, properties {padding=<ltrb 0 0 0 0>}}
  }
  TXT_MeshName {
   widgetClassPath=&TextBlock
   bIsVariable=true
   parent=BTN_MeshName
   widget=TXT_MeshName
   properties {
    minDesiredWidth=300
    text=Mesh Name
    font {
     bForceMonospaced=false
     bMaterialIsStencil=false
     fontMaterial=null
     fontObject=&Roboto
     letterSpacing=0
     monospacedWidth=1
     size=11
     skewAmount=0
     typefaceFontName=Bold
     outlineSettings {
      bApplyOutlineToDropShadows=false
      bMiteredCorners=false
      bSeparateFillAlpha=false
      outlineMaterial=null
      outlineSize=0
      outlineColor=<rgba 0 0 0 1>
     }
    }
    margin=<ltrb 4 0 0 0>
   }
   slot {class=ButtonSlot, properties {horizontalAlignment=HAlign_Fill, padding=<ltrb 0 0 0 0>}}
  }
  HB_LODColumns {
   widgetClassPath=&HorizontalBox
   bIsVariable=true
   parent=LODRowRoot
   widget=HB_LODColumns
   slot {class=HorizontalBoxSlot, properties {verticalAlignment=VAlign_Center}}
  }
  Border_1 {
   widgetClassPath=&Border
   bIsVariable=true
   parent=LODRowRoot
   widget=Border_1
   properties {
    toolTipText="Saves the Mesh\r\n* Prefix = The Mesh is Edited with Unsaved changes"
    verticalAlignment=VAlign_Center
    background {
     bIsDynamicallyLoaded=false
     drawAs=Image
     imageType=FullColor
     mirroring=NoMirror
     resourceName=None
     resourceObject=null
     tiling=NoTile
     imageSize=<0 0>
     margin=<ltrb 0 0 0 0>
     outlineSettings {
      bUseBrushTransparency=false
      roundingType=HalfHeightRadius
      width=0
      color {colorUseRule=UseColor_Specified, specifiedColor=<rgba 0 0 0 0>}
      cornerRadii=<0 0 0 0>
     }
     tintColor {colorUseRule=UseColor_Specified, specifiedColor=<rgba 1 1 1 1>}
     uVRegion {bIsValid=false, max=<0 0>, min=<0 0>}
    }
    brushColor=<rgba 0.0417 0.0417 0.0417 1>
    padding=<ltrb 4 3 4 3>
   }
   slot {class=HorizontalBoxSlot, properties {verticalAlignment=VAlign_Center, padding=<ltrb 2 0 0 0>}}
  }
  BTN_Save {
   widgetClassPath=&Button
   parent=Border_1
   widget=BTN_Save
   properties {widgetStyle {
     clickedSlateSound {resourceObject=null}
     disabled {
      bIsDynamicallyLoaded=false
      drawAs=NoDrawType
      imageType=NoImage
      mirroring=NoMirror
      resourceName=None
      resourceObject=null
      tiling=NoTile
      imageSize=<0 0>
      margin=<ltrb 0 0 0 0>
      outlineSettings {
       bUseBrushTransparency=false
       roundingType=HalfHeightRadius
       width=0
       color {colorUseRule=UseColor_Specified, specifiedColor=<rgba 0 0 0 0>}
       cornerRadii=<0 0 0 0>
      }
      tintColor {colorUseRule=UseColor_Specified, specifiedColor=<rgba 1 1 1 1>}
      uVRegion {bIsValid=false, max=<0 0>, min=<0 0>}
     }
     disabledForeground {colorUseRule=UseColor_Specified, specifiedColor=<rgba 0.5271 0.5271 0.5271 1>}
     hovered {
      bIsDynamicallyLoaded=false
      drawAs=RoundedBox
      imageType=NoImage
      mirroring=NoMirror
      resourceName=None
      resourceObject=null
      tiling=NoTile
      imageSize=<32 32>
      margin=<ltrb 0 0 0 0>
      outlineSettings {
       bUseBrushTransparency=true
       roundingType=FixedRadius
       width=0
       color {colorUseRule=UseColor_Specified, specifiedColor=<rgba 0 0 0 0>}
       cornerRadii=<4 4 4 4>
      }
      tintColor {colorUseRule=UseColor_Specified, specifiedColor=<rgba 0.0953 0.0953 0.0953 1>}
      uVRegion {bIsValid=false, max=<0 0>, min=<0 0>}
     }
     hoveredForeground {colorUseRule=UseColor_Specified, specifiedColor=<rgba 0 1 1 1>}
     hoveredSlateSound {resourceObject=null}
     normal {
      bIsDynamicallyLoaded=false
      drawAs=RoundedBox
      imageType=NoImage
      mirroring=NoMirror
      resourceName=None
      resourceObject=null
      tiling=NoTile
      imageSize=<32 32>
      margin=<ltrb 0 0 0 0>
      outlineSettings {
       bUseBrushTransparency=true
       roundingType=FixedRadius
       width=0
       color {colorUseRule=UseColor_Specified, specifiedColor=<rgba 0 0 0 0>}
       cornerRadii=<4 4 4 4>
      }
      tintColor {colorUseRule=UseColor_Specified, specifiedColor=<rgba 0.0438 0.0438 0.0438 1>}
      uVRegion {bIsValid=false, max=<0 0>, min=<0 0>}
     }
     normalForeground {colorUseRule=UseColor_Specified, specifiedColor=<rgba 0 1 1 1>}
     normalPadding=<ltrb 0 0 0 0>
     pressed {
      bIsDynamicallyLoaded=false
      drawAs=RoundedBox
      imageType=NoImage
      mirroring=NoMirror
      resourceName=None
      resourceObject=null
      tiling=NoTile
      imageSize=<32 32>
      margin=<ltrb 0 0 0 0>
      outlineSettings {
       bUseBrushTransparency=true
       roundingType=FixedRadius
       width=0
       color {colorUseRule=UseColor_Specified, specifiedColor=<rgba 0 0 0 0>}
       cornerRadii=<4 4 4 4>
      }
      tintColor {colorUseRule=UseColor_Specified, specifiedColor=<rgba 0.0284 0.0284 0.0284 1>}
      uVRegion {bIsValid=false, max=<0 0>, min=<0 0>}
     }
     pressedForeground {colorUseRule=UseColor_Specified, specifiedColor=<rgba 0 1 1 1>}
     pressedPadding=<ltrb 0 0 0 0>
     pressedSlateSound {resourceObject=null}
    }}
   slot {class=BorderSlot, properties {verticalAlignment=VAlign_Center, padding=<ltrb 0 0 0 0>}}
  }
  TXT_SaveLabel {
   widgetClassPath=&TextBlock
   bIsVariable=true
   parent=BTN_Save
   widget=TXT_SaveLabel
   properties {
    justification=Center
    minDesiredWidth=52
    text=Save
    font {
     bForceMonospaced=false
     bMaterialIsStencil=false
     fontMaterial=null
     fontObject=&Roboto
     letterSpacing=0
     monospacedWidth=1
     size=11
     skewAmount=0
     typefaceFontName=Bold
     outlineSettings {
      bApplyOutlineToDropShadows=false
      bMiteredCorners=false
      bSeparateFillAlpha=false
      outlineMaterial=null
      outlineSize=0
      outlineColor=<rgba 0 0 0 1>
     }
    }
   }
   slot {class=ButtonSlot, properties {horizontalAlignment=HAlign_Fill, verticalAlignment=VAlign_Fill, padding=<ltrb 0 0 0 0>}}
  }
 ]
}
